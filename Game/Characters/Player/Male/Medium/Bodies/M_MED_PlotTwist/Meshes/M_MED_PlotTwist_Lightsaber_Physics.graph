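# Physics asset for the dangling lightsaber on the PlotTwist medium male player body, simulating it against the hips and left thigh.
# Capsule bodies cover the saber, a tiny saber root, the pelvis and the left upper thigh twist bone.
# The root joins the pelvis through an angularly free constraint, while the saber hangs from the root with tight swing and almost no twist.
# A further constraint links the thigh to the pelvis.
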PhysicsAsset /Game/Characters/Player/Male/Medium/Bodies/M_MED_PlotTwist/Meshes/M_MED_PlotTwist_Lightsaber_Physics
bodies {
 dyn_lightsaber [{
   shapeType=Capsule
   length=22.4341
   radius=2.8331
   shapeName=dyn_lightsaber_capsule
   center=<8.6596 -0.5349 0.3548>
   rotation=<rot 0 87.6661 -88.4716>
  }]
 dyn_lightsaber_root [{shapeType=Capsule, length=0.505, radius=0.505, shapeName=dyn_lightsaber_root_capsule, rotation=<rot 0 -90 0>}]
 pelvis [{shapeType=Capsule, length=23.5242, radius=11.3347}]
 thigh_twist_01_l [{shapeType=Capsule, length=27.864, radius=10.2275, center=<8.5663 0.4556 0.2999>, rotation=<rot -2.2659 -92.8855 -89.361>}]
}
constraints [
 {bone1Name=thigh_twist_01_l, bone2Name=pelvis, swing1LimitDegrees=45, swing2LimitDegrees=45, twistLimitDegrees=45}
 {bone1Name=dyn_lightsaber, bone2Name=dyn_lightsaber_root, swing1LimitDegrees=10, swing2LimitDegrees=10, twistLimitDegrees=1}
 {
  bone1Name=dyn_lightsaber_root
  bone2Name=pelvis
  swing1LimitDegrees=45
  swing1Motion=Free
  swing2LimitDegrees=45
  swing2Motion=Free
  twistLimitDegrees=45
  twistMotion=Free
 }
]
