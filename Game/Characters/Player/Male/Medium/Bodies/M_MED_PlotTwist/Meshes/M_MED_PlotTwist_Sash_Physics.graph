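# Physics asset for the PlotTwist medium male player body, defining bodies and joints for its sash and underskirt secondary motion.
# Bone chains use capsules and boxes, alongside pelvis and thigh capsules for the body.
# Tiny root capsules join to the pelvis with free swing and twist, except front sash roots, which attach to underskirt A roots.
# Tight angular limits down each chain restrict bending, with rear sashes looser than most skirt panels.
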PhysicsAsset /Game/Characters/Player/Male/Medium/Bodies/M_MED_PlotTwist/Meshes/M_MED_PlotTwist_Sash_Physics
bodies {
 dyn_front_sash_l_01 [{
   shapeType=Capsule
   length=4.8843
   radius=5.2167
   shapeName=dyn_front_sash_l_01_capsule
   center=<5.8903 -3.0956 -0.2009>
   rotation=<rot 0 -176.6732 -83.9965>
  }]
 dyn_front_sash_l_01_root [{shapeType=Capsule, length=0.505, radius=0.505, shapeName=dyn_front_sash_l_01_root_capsule, rotation=<rot 0 -90 0>}]
 dyn_front_sash_l_02 [{
   shapeType=Capsule
   length=8.8358
   radius=5.9522
   shapeName=dyn_front_sash_l_02_capsule
   center=<7.0395 -1.9848 -1.6392>
   rotation=<rot 0.1673 178.7799 -89.8>
  }]
 dyn_front_sash_r_01 [{
   shapeType=Capsule
   length=3.7951
   radius=6.631
   shapeName=dyn_front_sash_r_01_capsule
   center=<-5.9475 -0.2836 0.3433>
   rotation=<rot -1.0609 177.092 -101.4553>
  }]
 dyn_front_sash_r_01_root [{shapeType=Capsule, length=0.505, radius=0.505, shapeName=dyn_front_sash_r_01_root_capsule, rotation=<rot 0 -90 0>}]
 dyn_front_sash_r_02 [{
   shapeType=Capsule
   length=7.6387
   radius=6.9184
   shapeName=dyn_front_sash_r_02_capsule
   center=<-7.5509 2.1582 1.3935>
   rotation=<rot -0.3039 178.9989 -98.8203>
  }]
 dyn_rear_sash_01_l [{
   shapeType=Capsule
   length=4.0083
   radius=4.5759
   shapeName=dyn_rear_sash_01_l_capsule
   center=<6.1328 -0.776 3.3263>
   rotation=<rot 0.0019 176.9581 -51.2124>
  }]
 dyn_rear_sash_01_l_root [{shapeType=Capsule, length=0.505, radius=0.505, shapeName=dyn_rear_sash_01_l_root_capsule, rotation=<rot 0 -90 0>}]
 dyn_rear_sash_01_r [{
   shapeType=Capsule
   length=4.8522
   radius=4.939
   shapeName=dyn_rear_sash_01_r_capsule
   center=<8.1174 0.3509 3.5629>
   rotation=<rot -0.0002 2.0284 -67.1434>
  }]
 dyn_rear_sash_01_r_root [{shapeType=Capsule, length=0.505, radius=0.505, shapeName=dyn_rear_sash_01_r_root_capsule, rotation=<rot 0 -90 0>}]
 dyn_rear_sash_02_l [{
   shapeType=Capsule
   length=8.5416
   radius=3.1857
   shapeName=dyn_rear_sash_02_l_capsule
   center=<3.6585 0.0253 4.1335>
   rotation=<rot -0.0001 -3.319 -95.3183>
  }]
 dyn_rear_sash_02_r [{
   shapeType=Capsule
   length=8.0378
   radius=3.4992
   shapeName=dyn_rear_sash_02_r_capsule
   center=<5.1398 -0.0839 5.5899>
   rotation=<rot -0.0006 2.8349 -86.3846>
  }]
 dyn_rear_sash_03_l [{
   shapeType=Capsule
   length=17.0178
   radius=4.1794
   shapeName=dyn_rear_sash_03_l_capsule
   center=<3.2247 -0.2045 4.9533>
   rotation=<rot 0 -1.0099 -89.358>
  }]
 dyn_rear_sash_03_r [{
   shapeType=Capsule
   length=15.2138
   radius=5.4325
   shapeName=dyn_rear_sash_03_r_capsule
   center=<4.6899 -1.9937 8.1017>
   rotation=<rot -0.0003 0 -93.5605>
  }]
 dyn_rear_sash_04_l [{
   shapeType=Capsule
   length=13.9753
   radius=6.1441
   shapeName=dyn_rear_sash_04_l_capsule
   center=<8.6606 -0.7317 6.9376>
   rotation=<rot 0 176.1804 -87.8266>
  }]
 dyn_rear_sash_04_r [{
   shapeType=Capsule
   length=19.3503
   radius=6.7066
   shapeName=dyn_rear_sash_04_l_capsule
   center=<7.717 -0.3691 10.5461>
   rotation=<rot -0.0001 -3.5277 -90>
  }]
 dyn_underskirt_a_l_01_root [{shapeType=Capsule, length=1, radius=1, shapeName=dyn_underskirt_a_l_01_root_capsule}]
 dyn_underskirt_a_l_02 [{
   shapeType=Box
   extentX=13.9375
   extentY=15.8125
   extentZ=9.25
   shapeName=dyn_underskirt_a_l_02_box
   center=<6.6007 0 1.0752>
   rotation=<rot -2.2435 0 0>
  }]
 dyn_underskirt_a_l_03 [{shapeType=Box, extentX=13.8125, extentY=24.0625, extentZ=11.0625, shapeName=dyn_underskirt_a_l_03_box, center=<7.6039 0 1.2437>}]
 dyn_underskirt_a_r_01_root [{shapeType=Capsule, length=1, radius=1, shapeName=dyn_underskirt_a_r_01_root_capsule}]
 dyn_underskirt_a_r_02 [{
   shapeType=Box
   extentX=13.9375
   extentY=15.8125
   extentZ=12
   shapeName=dyn_underskirt_a_l_02_box
   center=<-6.448 0 0.4227>
   rotation=<rot -6.2628 0 180>
  }]
 dyn_underskirt_a_r_03 [{
   shapeType=Box
   extentX=13.8125
   extentY=24.5
   extentZ=7.375
   shapeName=dyn_underskirt_a_l_03_box
   center=<-7.4119 5.8482 1.3241>
   rotation=<rot -6.1919 0 180>
  }]
 dyn_underskirt_c_l_01 [{shapeType=Capsule, length=4, radius=2, shapeName=dyn_underskirt_c_l_01_capsule, center=<10.8112 2.5879 -2.0498>, rotation=<rot 0 0 90>}]
 dyn_underskirt_c_l_01_root [{shapeType=Capsule, length=0.505, radius=0.505, shapeName=dyn_underskirt_c_l_01_root_capsule, rotation=<rot 0 -90 0>}]
 dyn_underskirt_c_l_02 [{
   shapeType=Capsule
   length=8.6875
   radius=4.125
   shapeName=dyn_underskirt_c_l_02_capsule
   center=<5.4743 0.0457 -0.3607>
   rotation=<rot 0 0 90>
  }]
 dyn_underskirt_c_l_03 [{
   shapeType=Capsule
   length=10.3125
   radius=3.875
   shapeName=dyn_underskirt_c_l_03_capsule
   center=<5.6062 -1.7052 -1.7535>
   rotation=<rot 0 0 90>
  }]
 dyn_underskirt_c_r_01 [{shapeType=Capsule, length=4, radius=2, shapeName=dyn_underskirt_c_r_01_capsule, center=<-2.3492 2.6793 -0.2373>, rotation=<rot 0 0 90>}]
 dyn_underskirt_c_r_01_root [{shapeType=Capsule, length=0.505, radius=0.505, shapeName=dyn_underskirt_c_r_01_root_capsule, rotation=<rot 0 -90 0>}]
 dyn_underskirt_c_r_02 [{shapeType=Capsule, length=4, radius=2, shapeName=dyn_underskirt_c_r_02_capsule, center=<-5.0873 1.8631 0.6041>, rotation=<rot 0 0 90>}]
 dyn_underskirt_c_r_03 [{
   shapeType=Capsule
   length=11.1875
   radius=4.0625
   shapeName=dyn_underskirt_c_r_03_capsule
   center=<-5.9512 1.4455 0.7854>
   rotation=<rot 0 0 109.277>
  }]
 dyn_underskirt_c_rear_01 [{shapeType=Capsule, length=4, radius=2, shapeName=dyn_underskirt_c_rear_01_capsule, center=<2.7956 -0.2518 -0.4839>, rotation=<rot 0 0 90>}]
 dyn_underskirt_c_rear_01_root [{shapeType=Capsule, length=0.505, radius=0.505, shapeName=dyn_underskirt_c_rear_01_root_capsule, rotation=<rot 0 -90 0>}]
 dyn_underskirt_c_rear_02 [
  {shapeType=Capsule, length=4, radius=2, shapeName=dyn_underskirt_c_rear_02_capsule, center=<4.5247 0 -0.7907>, rotation=<rot 0 0 90>}
  {
   shapeType=Capsule
   length=12.4062
   radius=3.1875
   shapeName=dyn_underskirt_c_rear_02_capsule
   center=<-0.0627 0 6.4567>
   rotation=<rot -89.4434 0 0>
  }
 ]
 dyn_underskirt_c_rear_03 [
  {
   shapeType=Capsule
   length=8.4062
   radius=3.6562
   shapeName=dyn_underskirt_c_rear_03_capsule
   center=<6.9068 0 9.8124>
   rotation=<rot -89.1567 0 0>
  }
  {shapeType=Box, extentX=14.25, extentY=20.5625, extentZ=9.75, shapeName=dyn_underskirt_c_rear_03_box, center=<4.7002 0 -1.0001>}
 ]
 dyn_underskirt_d_l_01 [{
   shapeType=Capsule
   length=9.0938
   radius=2.2188
   shapeName=dyn_underskirt_d_l_01_capsule
   center=<3.3351 1.8544 -1.0638>
   rotation=<rot -0.0001 0 107.101>
  }]
 dyn_underskirt_d_l_01_root [{shapeType=Capsule, length=0.505, radius=0.505, shapeName=dyn_underskirt_d_l_01_root_capsule, rotation=<rot 0 -90 0>}]
 dyn_underskirt_d_l_02 [{
   shapeType=Box
   extentX=14.125
   extentY=32.125
   extentZ=6.0625
   shapeName=dyn_underskirt_d_l_02_box
   center=<5.1331 0.0566 0.303>
   rotation=<rot -0.6578 -0.1228 10.5822>
  }]
 dyn_underskirt_d_l_03 [{
   shapeType=Box
   extentX=14.625
   extentY=44.3125
   extentZ=7.3125
   shapeName=dyn_underskirt_d_l_03_box
   center=<5.2545 0.0653 0.3419>
   rotation=<rot -1.3661 -0.261 10.8221>
  }]
 dyn_underskirt_d_r_01 [{
   shapeType=Capsule
   length=9.0938
   radius=2.2188
   shapeName=dyn_underskirt_d_r_01_capsule
   center=<-4.7456 -0.6421 -0.8648>
   rotation=<rot -0.0001 0 98.3633>
  }]
 dyn_underskirt_d_r_01_root [{shapeType=Capsule, length=0.505, radius=0.505, shapeName=dyn_underskirt_d_r_01_root_capsule, rotation=<rot 0 -90 0>}]
 dyn_underskirt_d_r_02 [{
   shapeType=Box
   extentX=14.125
   extentY=33.9375
   extentZ=7.3125
   shapeName=dyn_underskirt_d_l_02_box
   center=<-5.1514 -0.2913 -1.6552>
   rotation=<rot -0.8374 -0.156 -165.6892>
  }]
 dyn_underskirt_d_r_03 [{
   shapeType=Box
   extentX=14.625
   extentY=40.5
   extentZ=8.8125
   shapeName=dyn_underskirt_d_l_03_box
   center=<-5.3319 -0.3674 -2.0617>
   rotation=<rot -1.3661 -0.261 -169.1779>
  }]
 pelvis [{shapeType=Capsule, length=22.2935, radius=8.3712, center=<0 -0.1642 0>}]
 thigh_twist_01_l [{shapeType=Capsule, length=33.3352, radius=6.6816, center=<1.2282 0.5714 0>, rotation=<rot 90 0 0>}]
 thigh_twist_01_r [{shapeType=Capsule, length=33.3352, radius=6.6816, center=<-1.2282 0.5714 0>, rotation=<rot 90 0 180>}]
}
constraints [
 {bone1Name=thigh_twist_01_l, bone2Name=pelvis, swing1LimitDegrees=45, swing2LimitDegrees=45, twistLimitDegrees=45}
 {bone1Name=thigh_twist_01_r, bone2Name=pelvis, swing1LimitDegrees=45, swing2LimitDegrees=45, twistLimitDegrees=45}
 {
  bone1Name=dyn_underskirt_c_l_01_root
  bone2Name=pelvis
  swing1LimitDegrees=45
  swing1Motion=Free
  swing2LimitDegrees=45
  swing2Motion=Free
  twistLimitDegrees=45
  twistMotion=Free
 }
 {bone1Name=dyn_underskirt_c_l_01, bone2Name=dyn_underskirt_c_l_01_root, swing1LimitDegrees=8, swing2LimitDegrees=8, twistLimitDegrees=8}
 {bone1Name=dyn_underskirt_c_l_02, bone2Name=dyn_underskirt_c_l_01, swing1LimitDegrees=8, swing2LimitDegrees=8, twistLimitDegrees=8}
 {bone1Name=dyn_underskirt_c_l_03, bone2Name=dyn_underskirt_c_l_02, swing1LimitDegrees=8, swing2LimitDegrees=8, twistLimitDegrees=8}
 {
  bone1Name=dyn_underskirt_c_r_01_root
  bone2Name=pelvis
  swing1LimitDegrees=45
  swing1Motion=Free
  swing2LimitDegrees=45
  swing2Motion=Free
  twistLimitDegrees=45
  twistMotion=Free
 }
 {bone1Name=dyn_underskirt_c_r_01, bone2Name=dyn_underskirt_c_r_01_root, swing1LimitDegrees=8, swing2LimitDegrees=8, twistLimitDegrees=8}
 {bone1Name=dyn_underskirt_c_r_02, bone2Name=dyn_underskirt_c_r_01, swing1LimitDegrees=8, swing2LimitDegrees=8, twistLimitDegrees=8}
 {bone1Name=dyn_underskirt_c_r_03, bone2Name=dyn_underskirt_c_r_02, swing1LimitDegrees=8, swing2LimitDegrees=8, twistLimitDegrees=8}
 {
  bone1Name=dyn_underskirt_c_rear_01_root
  bone2Name=pelvis
  swing1LimitDegrees=45
  swing1Motion=Free
  swing2LimitDegrees=45
  swing2Motion=Free
  twistLimitDegrees=45
  twistMotion=Free
 }
 {bone1Name=dyn_underskirt_c_rear_01, bone2Name=dyn_underskirt_c_rear_01_root, swing1LimitDegrees=8, swing2LimitDegrees=8, twistLimitDegrees=8}
 {bone1Name=dyn_underskirt_c_rear_02, bone2Name=dyn_underskirt_c_rear_01, swing1LimitDegrees=2, swing2LimitDegrees=2, twistLimitDegrees=2}
 {bone1Name=dyn_underskirt_c_rear_03, bone2Name=dyn_underskirt_c_rear_02, swing1LimitDegrees=2, swing2LimitDegrees=2, twistLimitDegrees=2}
 {
  bone1Name=dyn_underskirt_d_l_01_root
  bone2Name=pelvis
  swing1LimitDegrees=45
  swing1Motion=Free
  swing2LimitDegrees=45
  swing2Motion=Free
  twistLimitDegrees=45
  twistMotion=Free
 }
 {bone1Name=dyn_underskirt_d_l_01, bone2Name=dyn_underskirt_d_l_01_root, swing1LimitDegrees=8, swing2LimitDegrees=8, twistLimitDegrees=8}
 {bone1Name=dyn_underskirt_d_l_02, bone2Name=dyn_underskirt_d_l_01, swing1LimitDegrees=8, swing2LimitDegrees=8, twistLimitDegrees=8}
 {bone1Name=dyn_underskirt_d_l_03, bone2Name=dyn_underskirt_d_l_02, swing1LimitDegrees=8, swing2LimitDegrees=8, twistLimitDegrees=8}
 {
  bone1Name=dyn_underskirt_d_r_01_root
  bone2Name=pelvis
  swing1LimitDegrees=45
  swing1Motion=Free
  swing2LimitDegrees=45
  swing2Motion=Free
  twistLimitDegrees=45
  twistMotion=Free
 }
 {bone1Name=dyn_underskirt_d_r_01, bone2Name=dyn_underskirt_d_r_01_root, swing1LimitDegrees=8, swing2LimitDegrees=8, twistLimitDegrees=8}
 {
  bone1Name=dyn_underskirt_a_l_01_root
  bone2Name=pelvis
  swing1LimitDegrees=45
  swing1Motion=Free
  swing2LimitDegrees=45
  swing2Motion=Free
  twistLimitDegrees=45
  twistMotion=Free
 }
 {
  bone1Name=dyn_underskirt_a_r_01_root
  bone2Name=pelvis
  swing1LimitDegrees=45
  swing1Motion=Free
  swing2LimitDegrees=45
  swing2Motion=Free
  twistLimitDegrees=45
  twistMotion=Free
 }
 {
  bone1Name=dyn_rear_sash_01_l_root
  bone2Name=pelvis
  swing1LimitDegrees=45
  swing1Motion=Free
  swing2LimitDegrees=45
  swing2Motion=Free
  twistLimitDegrees=45
  twistMotion=Free
 }
 {
  bone1Name=dyn_rear_sash_01_r_root
  bone2Name=pelvis
  swing1LimitDegrees=45
  swing1Motion=Free
  swing2LimitDegrees=45
  swing2Motion=Free
  twistLimitDegrees=45
  twistMotion=Free
 }
 {bone1Name=dyn_rear_sash_01_l, bone2Name=dyn_rear_sash_01_l_root, swing1LimitDegrees=7, swing2LimitDegrees=20, twistLimitDegrees=12}
 {bone1Name=dyn_rear_sash_02_l, bone2Name=dyn_rear_sash_01_l, swing1LimitDegrees=7, swing2LimitDegrees=20, twistLimitDegrees=12}
 {bone1Name=dyn_rear_sash_03_l, bone2Name=dyn_rear_sash_02_l, swing1LimitDegrees=7, swing2LimitDegrees=20, twistLimitDegrees=12}
 {bone1Name=dyn_rear_sash_04_l, bone2Name=dyn_rear_sash_03_l, swing1LimitDegrees=7, swing2LimitDegrees=20, twistLimitDegrees=12}
 {bone1Name=dyn_rear_sash_01_r, bone2Name=dyn_rear_sash_01_r_root, swing1LimitDegrees=7, swing2LimitDegrees=12, twistLimitDegrees=12}
 {bone1Name=dyn_rear_sash_02_r, bone2Name=dyn_rear_sash_01_r, swing1LimitDegrees=7, swing2LimitDegrees=16, twistLimitDegrees=12}
 {bone1Name=dyn_rear_sash_03_r, bone2Name=dyn_rear_sash_02_r, swing1LimitDegrees=7, swing2LimitDegrees=20, twistLimitDegrees=12}
 {bone1Name=dyn_underskirt_d_r_03, bone2Name=dyn_underskirt_d_r_02, swing1LimitDegrees=14, swing2LimitDegrees=14, twistLimitDegrees=14}
 {bone1Name=dyn_underskirt_d_r_02, bone2Name=dyn_underskirt_d_r_01, swing1LimitDegrees=8, swing2LimitDegrees=8, twistLimitDegrees=8}
 {bone1Name=dyn_rear_sash_04_r, bone2Name=dyn_rear_sash_03_r, swing1LimitDegrees=7, swing2LimitDegrees=20, twistLimitDegrees=12}
 {bone1Name=dyn_underskirt_a_l_02, bone2Name=dyn_underskirt_a_l_01_root, swing1LimitDegrees=3, swing2LimitDegrees=2, twistLimitDegrees=5}
 {bone1Name=dyn_underskirt_a_l_03, bone2Name=dyn_underskirt_a_l_02, swing1LimitDegrees=3, swing2LimitDegrees=2, twistLimitDegrees=5}
 {bone1Name=dyn_underskirt_a_r_03, bone2Name=dyn_underskirt_a_r_02, swing1LimitDegrees=3, swing2LimitDegrees=3, twistLimitDegrees=5}
 {bone1Name=dyn_underskirt_a_r_02, bone2Name=dyn_underskirt_a_r_01_root, swing1LimitDegrees=3, swing2LimitDegrees=3, twistLimitDegrees=5}
 {
  bone1Name=dyn_front_sash_l_01_root
  bone2Name=dyn_underskirt_a_l_01_root
  swing1LimitDegrees=45
  swing1Motion=Free
  swing2LimitDegrees=45
  swing2Motion=Free
  twistLimitDegrees=45
  twistMotion=Free
 }
 {bone1Name=dyn_front_sash_l_01, bone2Name=dyn_front_sash_l_01_root, swing1LimitDegrees=8, swing2LimitDegrees=18, twistLimitDegrees=1}
 {bone1Name=dyn_front_sash_l_02, bone2Name=dyn_front_sash_l_01, swing1LimitDegrees=8, swing2LimitDegrees=18, twistLimitDegrees=1}
 {
  bone1Name=dyn_front_sash_r_01_root
  bone2Name=dyn_underskirt_a_r_01_root
  swing1LimitDegrees=45
  swing1Motion=Free
  swing2LimitDegrees=45
  swing2Motion=Free
  twistLimitDegrees=45
  twistMotion=Free
 }
 {bone1Name=dyn_front_sash_r_01, bone2Name=dyn_front_sash_r_01_root, swing1LimitDegrees=4, swing2LimitDegrees=18, twistLimitDegrees=1}
 {bone1Name=dyn_front_sash_r_02, bone2Name=dyn_front_sash_r_01, swing1LimitDegrees=8, swing2LimitDegrees=18, twistLimitDegrees=1}
]
