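# Anim Blueprint for the M_MED_PlotTwist player body that adds secondary motion to its robe, sashes, belt gear and lightsaber.
# The copied parent pose runs through weighted bone constraints and DeformRig and UnderSkirt Control Rigs, then two cached-pose sash RigidBody branches meet at a bool blend.
# Lightsaber physics, left-wrist AnimDynamics and crouch-only bone offsets follow.
# The EventGraph computes state-based gravity and force variables, which the AnimGraph's physics nodes do not read.
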AnimBlueprint /Game/Characters/Player/Male/Medium/Bodies/M_MED_PlotTwist/Meshes/M_MED_PlotTwist_AnimBP
target_skeleton=&/M_MED_PlotTwist_Skeleton
graphs [
 AnimGraph {nodes [
   n1 = RigidBody(ComponentPose=Local To Component(LocalPose=Use cached pose 'PostSkirtPhysics'())) {
    bOverrideWorldGravity=true
    overridePhysicsAsset=&/M_MED_PlotTwist_Lightsaber_Physics
    simulationSpace=BaseBoneSpace
    alphaBoolBlend {blendOption=Linear}
    baseBoneRef {boneName=pelvis}
    componentAppliedLinearAccClamp=<1500 1500 1500>
    componentLinearAccScale=<0.1 0.1 0.1>
    componentLinearVelScale=<0.1 0.1 0.1>
    overrideWorldGravity {z=-980}
    simSpaceSettings {maxAngularAcceleration=50, maxAngularVelocity=50, maxLinearAcceleration=50, maxLinearVelocity=50, worldAlpha=1}
   }
   n2 = Anim Dynamics - Bone: dyn_robe_wrist_l(ComponentPose=@n1, Alpha=1, ExternalForce="0,0,0") {
    angularSpringConstant=0.5
    bAngularSpring=true
    bLinearSpring=true
    bUseGravityOverride=true
    linearSpringConstant=0.5
    alphaBoolBlend {blendOption=Linear}
    boundBone {boneName=dyn_robe_wrist_l}
    componentAppliedLinearAccClamp=<3000 3000 3000>
    componentLinearAccScale=<0.3 0.3 0.3>
    componentLinearVelScale=<0.3 0.3 0.3>
    gravityOverride {z=-980}
    physicsBodyDefinitions [{
      boundBone {boneName=dyn_robe_wrist_l}
      boxExtents=<20 20 20>
      constraintSetup {
       angularLimitsMax=<2 2 2>
       angularLimitsMin=<-2 -2 -2>
       angularTarget=<0 0 0>
       linearAxesMax=<0.25 0.25 0.25>
       linearAxesMin=<-0.25 -0.25 -0.25>
      }
      localJointOffset=<0 0 -2>
     }]
    relativeSpaceBone {boneName=lowerarm_l}
   }
   n3 = Anim Dynamics - Bone: dyn_robe_wrist_end_l(ComponentPose=@n2, Alpha=1, ExternalForce="0,0,0") {
    angularSpringConstant=0.5
    bAngularSpring=true
    bLinearSpring=true
    bUseGravityOverride=true
    linearSpringConstant=0.5
    simulationSpace=BoneRelative
    alphaBoolBlend {blendOption=Linear}
    boundBone {boneName=dyn_robe_wrist_end_l}
    componentAppliedLinearAccClamp=<3000 3000 3000>
    componentLinearAccScale=<0.3 0.3 0.3>
    componentLinearVelScale=<0.3 0.3 0.3>
    gravityOverride {z=-980}
    physicsBodyDefinitions [{
      boundBone {boneName=dyn_robe_wrist_end_l}
      boxExtents=<20 20 20>
      constraintSetup {angularLimitsMax=<10 10 10>, angularLimitsMin=<-10 -10 -10>, angularTarget=<0 0 0>, linearAxesMax=<1 1 1>, linearAxesMin=<-1 -1 -1>}
      localJointOffset=<0 0 -5>
     }]
    relativeSpaceBone {boneName=dyn_robe_wrist_l}
   }
   n4 = Get bIsCrouching()
   n5 = "Transform (Modify) Bone - Bone: deform_front_r_pouch_belt"!(ComponentPose=@n3, bAlphaBoolEnabled:bool=@n4, Translation="0,-1.5,2.3", Rotation="0,0,0", Scale="1,1,1") -> (Pose) {
    alphaInputType=Bool
    translationMode=BMM_Additive
    alphaBoolBlend {blendOption=Linear}
    boneToModify {boneName=deform_front_r_pouch_belt}
    translation {y=-1.5, z=2.3}
   }
   n6 = "Transform (Modify) Bone - Bone: deform_r_canister_03"!(ComponentPose=@n5, bAlphaBoolEnabled:bool=@n4, Translation="-0.043,-1.617,0.162", Rotation="-5.633,21.823,0.106", Scale="1,1,1") -> (Pose) {
    alphaInputType=Bool
    rotationMode=BMM_Additive
    translationMode=BMM_Additive
    alphaBoolBlend {blendOption=Linear}
    boneToModify {boneName=deform_r_canister_03}
    rotation=<rot -5.633 21.823 0.106>
    translation=<-0.043 -1.617 0.162>
   }
   n7 = "Transform (Modify) Bone - Bone: deform_r_canister_04"!(ComponentPose=@n6, bAlphaBoolEnabled:bool=@n4, Translation="0.610,-1.598,0.471", Rotation="-2.941,15.262,3.970", Scale="1,1,1") -> (Pose) {
    alphaInputType=Bool
    rotationMode=BMM_Additive
    translationMode=BMM_Additive
    alphaBoolBlend {blendOption=Linear}
    boneToModify {boneName=deform_r_canister_04}
    rotation=<rot -2.941 15.262 3.97>
    translation=<0.61 -1.598 0.471>
   }
   n8 = "Transform (Modify) Bone - Bone: deform_r_canister_02"!(ComponentPose=@n7, bAlphaBoolEnabled:bool=@n4, Translation="-0.867,-0.583,0.144", Rotation="0,0,0", Scale="1,1,1") -> (Pose) {
    alphaInputType=Bool
    translationMode=BMM_Additive
    alphaBoolBlend {blendOption=Linear}
    boneToModify {boneName=deform_r_canister_02}
    translation=<-0.867 -0.583 0.144>
   }
   n9 = "Transform (Modify) Bone - Bone: deform_r_canister_01"!(ComponentPose=@n8, bAlphaBoolEnabled:bool=@n4, Translation="-0.765,-0.798,0.149", Rotation="0,0,0", Scale="1,1,1") -> (Pose) {
    alphaInputType=Bool
    translationMode=BMM_Additive
    alphaBoolBlend {blendOption=Linear}
    boneToModify {boneName=deform_r_canister_01}
    translation=<-0.765 -0.798 0.149>
   }
   n10 = "Transform (Modify) Bone - Bone: deform_sash_01_r_pelvis_correction"!(ComponentPose=@n9, bAlphaBoolEnabled:bool=Get bIsCrouching(), Translation="2,0,0", Rotation="-50,0,0", Scale="1,1,1") -> (Pose) {
    alphaInputType=Bool
    rotationMode=BMM_Additive
    scaleMode=BMM_Additive
    translationMode=BMM_Additive
    alphaBoolBlend {blendOption=Linear}
    boneToModify {boneName=deform_sash_01_r_pelvis_correction}
    rotation {pitch=-50}
    translation {x=2}
   }
   Output Pose(Result=Component To Local(ComponentPose=@n10)) {layerGroup=None}
   n11 = "Constraint - deform_rear_r_pouch_belt (PPPP)"!(
    ComponentPose=Local To Component(LocalPose=Copy Pose From Mesh() {bUseAttachedParent=true})
    Alpha=1
    ConstraintWeights_0=0.5659
    ConstraintWeights_1:float=0.2375
    ConstraintWeights_2:float=0.1965
    ConstraintWeights_3:float=0
   ) -> (Pose) {
    alphaBoolBlend {blendOption=Linear}
    boneToModify {boneName=deform_rear_r_pouch_belt}
    constraintSetup [{targetBone {boneName=pelvis}}, {targetBone {boneName=spine_02}}, {targetBone {boneName=spine_01}}, {targetBone {boneName=spine_04}}]
    constraintWeights [1,1,1,1]
   }
   n12 = "Constraint - deform_rear_l_pouch_belt (PPPP)"!(
    ComponentPose=@n11
    Alpha=1
    ConstraintWeights_0=0.6513
    ConstraintWeights_1:float=0.1844
    ConstraintWeights_2:float=0.1638
    ConstraintWeights_3:float=0.0005
   ) -> (Pose) {
    alphaBoolBlend {blendOption=Linear}
    boneToModify {boneName=deform_rear_l_pouch_belt}
    constraintSetup [{targetBone {boneName=pelvis}}, {targetBone {boneName=spine_02}}, {targetBone {boneName=spine_01}}, {targetBone {boneName=spine_04}}]
    constraintWeights [1,1,1,1]
   }
   n13 = "Constraint - deform_lightsaber_holster_belt (PPPPP)"!(
    ComponentPose=@n12
    Alpha=1
    ConstraintWeights_0=0.745
    ConstraintWeights_1:float=0.1354
    ConstraintWeights_2:float=0.1165
    ConstraintWeights_3:float=0.0016
    ConstraintWeights_4:float=0.0015
   ) -> (Pose) {
    alphaBoolBlend {blendOption=Linear}
    boneToModify {boneName=deform_lightsaber_holster_belt}
    constraintSetup [
     {targetBone {boneName=pelvis}}
     {targetBone {boneName=spine_02}}
     {targetBone {boneName=spine_01}}
     {targetBone {boneName=deform_groin_l}}
     {targetBone {boneName=thigh_twist_01_l}}
    ]
    constraintWeights [1,1,1,1,1]
   }
   n14 = "Constraint - deform_front_r_pouch_belt_root (PPP)"!(ComponentPose=@n13, Alpha=1, ConstraintWeights_0=0.6455, ConstraintWeights_1:float=0.1916, ConstraintWeights_2:float=0.1628) -> (Pose) {
    alphaBoolBlend {blendOption=Linear}
    boneToModify {boneName=deform_front_r_pouch_belt_root}
    constraintSetup [{targetBone {boneName=pelvis}}, {targetBone {boneName=spine_02}}, {targetBone {boneName=spine_01}}]
    constraintWeights [1,1,1]
   }
   n15 = "Constraint - deform_l_canister_01 (PPPPP)"!(
    ComponentPose=@n14
    Alpha=1
    ConstraintWeights_0=0.7371
    ConstraintWeights_1:float=0.142
    ConstraintWeights_2:float=0.1206
    ConstraintWeights_3:float=0.0003
    ConstraintWeights_4:float=0.0001
   ) -> (Pose) {
    alphaBoolBlend {blendOption=Linear}
    boneToModify {boneName=deform_l_canister_01}
    constraintSetup [
     {targetBone {boneName=pelvis}}
     {targetBone {boneName=spine_02}}
     {targetBone {boneName=spine_01}}
     {targetBone {boneName=thigh_twist_01_l}}
     {targetBone {boneName=deform_groin_l}}
    ]
    constraintWeights [1,1,1,1,1]
   }
   n16 = "Constraint - deform_l_canister_02 (PPPP)"!(
    ComponentPose=@n15
    Alpha=1
    ConstraintWeights_0=0.8248
    ConstraintWeights_1:float=0.0919
    ConstraintWeights_2:float=0.0784
    ConstraintWeights_3:float=0.005
   ) -> (Pose) {
    alphaBoolBlend {blendOption=Linear}
    boneToModify {boneName=deform_l_canister_02}
    constraintSetup [{targetBone {boneName=pelvis}}, {targetBone {boneName=spine_02}}, {targetBone {boneName=spine_01}}, {targetBone {boneName=thigh_twist_01_l}}]
    constraintWeights [1,1,1,1]
   }
   n17 = "Constraint - deform_l_canister_03 (PPP)"!(ComponentPose=@n16, Alpha=1, ConstraintWeights_0=0.7224, ConstraintWeights_1:float=0.1507, ConstraintWeights_2:float=0.1269) -> (Pose) {
    alphaBoolBlend {blendOption=Linear}
    boneToModify {boneName=deform_l_canister_03}
    constraintSetup [{targetBone {boneName=pelvis}}, {targetBone {boneName=spine_02}}, {targetBone {boneName=spine_01}}]
    constraintWeights [1,1,1]
   }
   n18 = "Constraint - deform_l_canister_04 (PPPP)"!(
    ComponentPose=@n17
    Alpha=1
    ConstraintWeights_0=0.8232
    ConstraintWeights_1:float=0.0948
    ConstraintWeights_2:float=0.0811
    ConstraintWeights_3:float=0.0009
   ) -> (Pose) {
    alphaBoolBlend {blendOption=Linear}
    boneToModify {boneName=deform_l_canister_04}
    constraintSetup [{targetBone {boneName=pelvis}}, {targetBone {boneName=spine_02}}, {targetBone {boneName=spine_01}}, {targetBone {boneName=thigh_twist_01_l}}]
    constraintWeights [1,1,1,1]
   }
   n19 = "Constraint - deform_r_canister_04 (PPPPP)"!(
    ComponentPose=@n18
    Alpha=1
    ConstraintWeights_0=0.7371
    ConstraintWeights_1:float=0.142
    ConstraintWeights_2:float=0.1206
    ConstraintWeights_3:float=0.0003
    ConstraintWeights_4:float=0.0001
   ) -> (Pose) {
    alphaBoolBlend {blendOption=Linear}
    boneToModify {boneName=deform_r_canister_04}
    constraintSetup [
     {targetBone {boneName=pelvis}}
     {targetBone {boneName=spine_02}}
     {targetBone {boneName=spine_01}}
     {targetBone {boneName=thigh_twist_01_r}}
     {targetBone {boneName=deform_groin_r}}
    ]
    constraintWeights [1,1,1,1,1]
   }
   n20 = "Constraint - deform_r_canister_03 (PPPP)"!(
    ComponentPose=@n19
    Alpha=1
    ConstraintWeights_0=0.5972
    ConstraintWeights_1:float=0.2192
    ConstraintWeights_2:float=0.1836
    ConstraintWeights_3:float=0
   ) -> (Pose) {
    alphaBoolBlend {blendOption=Linear}
    boneToModify {boneName=deform_r_canister_03}
    constraintSetup [{targetBone {boneName=pelvis}}, {targetBone {boneName=spine_02}}, {targetBone {boneName=spine_01}}, {targetBone {boneName=spine_04}}]
    constraintWeights [1,1,1,1]
   }
   n21 = "Constraint - deform_r_canister_02 (PPPP)"!(
    ComponentPose=@n20
    Alpha=1
    ConstraintWeights_0=0.6695
    ConstraintWeights_1:float=0.1789
    ConstraintWeights_2:float=0.1515
    ConstraintWeights_3:float=0
   ) -> (Pose) {
    alphaBoolBlend {blendOption=Linear}
    boneToModify {boneName=deform_r_canister_02}
    constraintSetup [{targetBone {boneName=pelvis}}, {targetBone {boneName=spine_02}}, {targetBone {boneName=spine_01}}, {targetBone {boneName=spine_04}}]
    constraintWeights [1,1,1,1]
   }
   n22 = "Constraint - deform_r_canister_01 (PPPP)"!(
    ComponentPose=@n21
    Alpha=1
    ConstraintWeights_0=0.6695
    ConstraintWeights_1:float=0.1789
    ConstraintWeights_2:float=0.1515
    ConstraintWeights_3:float=0
   ) -> (Pose) {
    alphaBoolBlend {blendOption=Linear}
    boneToModify {boneName=deform_r_canister_01}
    constraintSetup [{targetBone {boneName=pelvis}}, {targetBone {boneName=spine_02}}, {targetBone {boneName=spine_01}}, {targetBone {boneName=spine_04}}]
    constraintWeights [1,1,1,1]
   }
   n23 = "Constraint - dyn_sash_01_l_root (RR)"!(ComponentPose=@n22, Alpha=1, ConstraintWeights_0=0.5, ConstraintWeights_1:float=0.5) -> (Pose) {
    alphaBoolBlend {blendOption=Linear}
    boneToModify {boneName=dyn_sash_01_l_root}
    constraintSetup [{transformType=Rotation, targetBone {boneName=thigh_twist_01_l}}, {transformType=Rotation, targetBone {boneName=pelvis}}]
    constraintWeights [1,1]
   }
   n24 = "Constraint - dyn_sash_01_r_root (RR)"!(ComponentPose=@n23, Alpha=1, ConstraintWeights_0=0.5, ConstraintWeights_1:float=0.5) -> (Pose) {
    alphaBoolBlend {blendOption=Linear}
    boneToModify {boneName=dyn_sash_01_r_root}
    constraintSetup [{transformType=Rotation, targetBone {boneName=thigh_twist_01_r}}, {transformType=Rotation, targetBone {boneName=pelvis}}]
    constraintWeights [1,1]
   }
   n25 = Control Rig!(
    Source=Component To Local(ComponentPose=@n24)
    Alpha
    upperarm_bk_bk_jnt:Transform=Get Upperarm Bk Bk Jnt() -> (Upperarm Bk Bk Jnt:Transform)
    upperarm_front_bk_jnt:Transform=Get Upperarm Front Bk Jnt() -> (Upperarm Front Bk Jnt:Transform)
    upperarm_bk_sash_clavicle:Transform=Get Upperarm Bk Sash Clavicle() -> (Upperarm Bk Sash Clavicle:Transform)
    upperarm_bk_sash_lowerback:Transform=Get Upperarm Bk Sash Lowerback() -> (Upperarm Bk Sash Lowerback:Transform)
    upperarm_up_sash_clavicle:Transform=Get Upperarm Up Sash Clavicle() -> (Upperarm Up Sash Clavicle:Transform)
    upperarm_fwd_sash_clavicle:Transform=Get Upperarm Fwd Sash Clavicle() -> (Upperarm Fwd Sash Clavicle:Transform)
    upperarm_down_sash_clavicle:Transform=Get Upperarm Down Sash Clavicle() -> (Upperarm Down Sash Clavicle:Transform)
    upperarm_down_sash_lowerback:Transform=Get Upperarm Down Sash Lowerback() -> (Upperarm Down Sash Lowerback:Transform)
    upperarm_down_45_sash_lowerback:Transform=Get Upperarm Down 45 Sash Lowerback() -> (Upperarm Down 45 Sash Lowerback:Transform)
    upperarm_raised:Transform=Get Upperarm Raised() -> (Upperarm Raised:Transform)
    is_skydiving:bool=Get bIsSkyDiving()
    upperarm_up_sash_clavicle_skydive:Transform=Get Upperarm Up Sash Clavicle Skydive() -> (Upperarm Up Sash Clavicle Skydive:Transform)
   ) -> (Pose) {
    alphaBoolBlend {blendOption=Linear}
    controlRigAssetReference {blueprintRigClass=&/M_MED_PlotTwist_DeformRig_ControlRig.M_MED_PlotTwist_DeformRig_ControlRig_C}
   }
   n26 = Control Rig!(
    Source=@n25
    Alpha
    underskirt_thigh_up_front:Transform=Get Underskirt Thigh Up Front() -> (Underskirt Thigh Up Front:Transform)
    underskirt_thigh_up_front_center:Transform=Get Underskirt Thigh Up Front Center() -> (Underskirt Thigh Up Front Center:Transform)
    underskirt_thigh_back_rear_center:Transform=Get Underskirt Thigh Back Rear Center() -> (Underskirt Thigh Back Rear Center:Transform)
    underskirt_thigh_back_rear:Transform=Get Underskirt Thigh Back Rear() -> (Underskirt Thigh Back Rear:Transform)
    underskirt_thigh_up_side:Transform=Get Underskirt Thigh Up Side() -> (Underskirt Thigh Up Side:Transform)
    lightsaber_thigh_up:Transform=Get Lightsaber Thigh Up() -> (Lightsaber Thigh Up:Transform)
    sash_thigh_up_front:Transform=Get Sash Thigh Up Front() -> (Sash Thigh Up Front:Transform)
    sash_thigh_up_opposite_front:Transform=Get Sash Thigh Up Opposite Front() -> (Sash Thigh Up Opposite Front:Transform)
    underskirt_thigh_up_opposite_front:Transform=Get Underskirt Thigh Up Opposite Front() -> (Underskirt Thigh Up Opposite Front:Transform)
    underskirt_thigh_up_thigh_side_inner_front:Transform=Get Underskirt Thigh Up Thigh Side Inner Front() -> (Underskirt Thigh Up Thigh Side Inner Front:Transform)
    underskirt_thigh_up_front_scale_end:Transform=Get Underskirt Thigh Up Front Scale End() -> (Underskirt Thigh Up Front Scale End:Transform)
    sash_thigh_up_thigh_side_inner_front_opposite:Transform=Get Sash Thigh Up Thigh Side Inner Front Opposite() -> (Sash Thigh Up Thigh Side Inner Front Opposite:Transform)
    thigh_back_underskirt:Transform=Get Thigh Back Underskirt() -> (Thigh Back Underskirt:Transform)
    thigh_back_rear_sash:Transform=Get Thigh Back Rear Sash() -> (Thigh Back Rear Sash:Transform)
   ) -> (Pose) {
    alphaBoolBlend {blendOption=Linear}
    controlRigAssetReference {blueprintRigClass=&/M_MED_PlotTwist_UnderSkirt_ControlRig.M_MED_PlotTwist_UnderSkirt_ControlRig_C}
   }
   Save cached pose 'PostControlRig'(Pose=@n26) {cacheName=PostControlRig}
   n27 = "Transform (Modify) Bone - Bone: dyn_front_sash_l_01_root"(
    ComponentPose=Local To Component(LocalPose=Use cached pose 'PostControlRig'())
    Alpha=1
    Translation="0,1.5,0"
    Rotation="-3.153,0.215,-4.723"
    Scale="1,1,1"
   ) {
    rotationMode=BMM_Additive
    translationMode=BMM_Additive
    alphaBoolBlend {blendOption=Linear}
    boneToModify {boneName=dyn_front_sash_l_01_root}
    rotation=<rot -3.153 0.215 -4.723>
    translation {y=1.5}
   }
   n28 = "Transform (Modify) Bone - Bone: dyn_front_sash_r_01_root"(ComponentPose=@n27, Alpha=1, Translation="0,1.5,0", Rotation="4.646,0.377,3.234", Scale="1,1,1") {
    rotationMode=BMM_Additive
    translationMode=BMM_Additive
    alphaBoolBlend {blendOption=Linear}
    boneToModify {boneName=dyn_front_sash_r_01_root}
    rotation=<rot 4.646 0.377 3.234>
    translation {y=1.5}
   }
   n29 = RigidBody!(ComponentPose=@n28, Alpha, OverrideWorldGravity:Vector="0,0,-980", ExternalForce) -> (Pose) {
    bOverrideWorldGravity=true
    overridePhysicsAsset=&/M_MED_PlotTwist_Sash_Frontend_Physics
    simulationSpace=BaseBoneSpace
    alphaBoolBlend {blendOption=Linear}
    baseBoneRef {boneName=pelvis}
    componentAppliedLinearAccClamp=<2000 2000 3000>
    componentLinearAccScale=<0.2 0.2 0.3>
    componentLinearVelScale=<0.2 0.2 0.3>
    overrideWorldGravity {z=-980}
    simSpaceSettings {maxAngularAcceleration=50, maxAngularVelocity=50, maxLinearAcceleration=50, maxLinearVelocity=50, worldAlpha=1}
   }
   n30 = RigidBody(ComponentPose=Local To Component(LocalPose=Use cached pose 'PostControlRig'()), OverrideWorldGravity="0,0,-980") {
    bOverrideWorldGravity=true
    overridePhysicsAsset=&/M_MED_PlotTwist_Sash_Physics
    simulationSpace=BaseBoneSpace
    alphaBoolBlend {blendOption=Linear}
    baseBoneRef {boneName=pelvis}
    componentAppliedLinearAccClamp=<2000 2000 3000>
    componentLinearAccScale=<0.2 0.2 0.3>
    componentLinearVelScale=<0.2 0.2 0.3>
    overrideWorldGravity {z=-980}
    simSpaceSettings {maxAngularAcceleration=50, maxAngularVelocity=50, maxLinearAcceleration=50, maxLinearVelocity=50, worldAlpha=1}
   }
   Save cached pose 'PostSkirtPhysics'(Pose=Blend Poses by bool(BlendPose_0=Component To Local(ComponentPose=@n29), BlendPose_1=Component To Local(ComponentPose=@n30)) {blendType=Linear}) {cacheName=PostSkirtPhysics}
  ]}
 EventGraph {nodes [
   Event BlueprintUpdateAnimation()
   > n1 = Gravity Override State Function DEPRECATED!(
    execute
    CharacterPartAnimInstance=Self-Reference()
    GameTimeInSeconds=GetGameTimeInSeconds()
    GravityOverrideParams_Skydive_GoingUp_28_#1:Gravity Override Setting Struct Structure=(WindFrequency_28_#2=(X=1.5,Y=7,Z=8),WindAmplitude_31_#3=(X=200,Y=-500,Z=400),WindOffset_32_#4=(X=0,Y=0,Z=0),GravityOverride_15_#5=(X=0,Y=-300,Z=1500),JointName_27_#6="root")
    GravityOverrideParams_Skydive_FloatingDown_29_#7:Gravity Override Setting Struct Structure=Get Gravity Override Params Skydive Floating Down() -> (Gravity Override Params Skydive Floating Down:Gravity Override Setting Struct Structure)
    GravityOverrideParams_Skydive_DivingDown_30_#8:Gravity Override Setting Struct Structure=(WindFrequency_28_#2=(X=1.5,Y=12,Z=10),WindAmplitude_31_#3=(X=50,Y=300,Z=100),WindOffset_32_#4=(X=0,Y=0,Z=0),GravityOverride_15_#5=(X=-1500,Y=0,Z=0),JointName_27_#6="pelvis")
    GravityOverrideParams_Skydive_Parachuting_31_#9:Gravity Override Setting Struct Structure=Get Gravity Override Params Skydive Parachuting() -> (Gravity Override Params Skydive Parachuting:Gravity Override Setting Struct Structure)
    GravityOverrideParams_OnGround_Standing_32_#10:Gravity Override Setting Struct Structure=(WindFrequency_28_#2=(X=0,Y=0,Z=0),WindAmplitude_31_#3=(X=0,Y=0,Z=0),WindOffset_32_#4=(X=0,Y=0,Z=0),GravityOverride_15_#5=(X=0,Y=0,Z=-980),JointName_27_#6="root")
    GravityOverrideParams_OnGround_Moving_33_#11:Gravity Override Setting Struct Structure=(WindFrequency_28_#2=(X=0,Y=0,Z=0),WindAmplitude_31_#3=(X=0,Y=0,Z=0),WindOffset_32_#4=(X=0,Y=0,Z=0),GravityOverride_15_#5=(X=0,Y=0,Z=-980),JointName_27_#6="root")
    GravityOverrideParams_EmoteOrMelee_34_#12:Gravity Override Setting Struct Structure=(WindFrequency_28_#2=(X=0,Y=0,Z=0),WindAmplitude_31_#3=(X=0,Y=0,Z=0),WindOffset_32_#4=(X=0,Y=0,Z=0),GravityOverride_15_#5=(X=0,Y=0,Z=-980),JointName_27_#6="root")
    GravityOverrideParams_Swimming_35_#13:Gravity Override Setting Struct Structure=(WindFrequency_28_#2=(X=0,Y=0,Z=0),WindAmplitude_31_#3=(X=0,Y=0,Z=0),WindOffset_32_#4=(X=0,Y=0,Z=0),GravityOverride_15_#5=(X=-800,Y=0,Z=0),JointName_27_#6="pelvis")
    GravityOverrideParams_Front_End_36_#14:Gravity Override Setting Struct Structure=(WindFrequency_28_#2=(X=0,Y=0,Z=0),WindAmplitude_31_#3=(X=0,Y=0,Z=0),WindOffset_32_#4=(X=0,Y=0,Z=0),GravityOverride_15_#5=(X=0,Y=0,Z=-980),JointName_27_#6="root")
    GravityOverrideParams_DBNO_40_#15:Gravity Override Setting Struct Structure=(WindFrequency_28_#2=(X=0,Y=0,Z=0),WindAmplitude_31_#3=(X=0,Y=0,Z=0),WindOffset_32_#4=(X=0,Y=0,Z=0),GravityOverride_15_#5=(X=-980,Y=400,Z=0),JointName_27_#6="pelvis")
    GravityOverrideParams_DBNO_Carried_41_#16:Gravity Override Setting Struct Structure=(WindFrequency_28_#2=(X=0,Y=0,Z=0),WindAmplitude_31_#3=(X=0,Y=0,Z=0),WindOffset_32_#4=(X=0,Y=0,Z=0),GravityOverride_15_#5=(X=-980,Y=500,Z=-500),JointName_27_#6="pelvis")
    GravityOverrideParams_OnGround_Falling_45_#17:Gravity Override Setting Struct Structure=(WindFrequency_28_#2=(X=0,Y=0,Z=0),WindAmplitude_31_#3=(X=0,Y=0,Z=0),WindOffset_32_#4=(X=0,Y=0,Z=0),GravityOverride_15_#5=(X=0,Y=0,Z=-980),JointName_27_#6="root")
    GravityOverrideParams_Ziplining_46_#18:Gravity Override Setting Struct Structure=Get Gravity Override Params Ziplining() -> (Gravity Override Params Ziplining:Gravity Override Setting Struct Structure)
    GravityOverrideParams_Skydive_Surfing_49_#19:Gravity Override Setting Struct Structure=(WindFrequency_28_#2=(X=1.5,Y=7,Z=8),WindAmplitude_31_#3=(X=200,Y=-500,Z=400),WindOffset_32_#4=(X=0,Y=0,Z=0),GravityOverride_15_#5=(X=500,Y=-300,Z=-980),JointName_27_#6="root")
    GravityOverrideParams_Riding_Standing_54_#20:Gravity Override Setting Struct Structure=(WindFrequency_28_#2=(X=0,Y=0,Z=0),WindAmplitude_31_#3=(X=0,Y=0,Z=0),WindOffset_32_#4=(X=0,Y=0,Z=0),GravityOverride_15_#5=(X=0,Y=-350,Z=-980),JointName_27_#6="root")
    GravityOverrideParams_Riding_Moving_55_#21:Gravity Override Setting Struct Structure=(WindFrequency_28_#2=(X=0,Y=0,Z=0),WindAmplitude_31_#3=(X=0,Y=0,Z=0),WindOffset_32_#4=(X=0,Y=0,Z=0),GravityOverride_15_#5=(X=0,Y=-450,Z=-450),JointName_27_#6="root")
    GravityOverrideParams_Riding_Moving_Targeting_58_#22:Gravity Override Setting Struct Structure=(WindFrequency_28_#2=(X=0,Y=0,Z=0),WindAmplitude_31_#3=(X=0,Y=0,Z=0),WindOffset_32_#4=(X=0,Y=0,Z=0),GravityOverride_15_#5=(X=0,Y=-200,Z=-980),JointName_27_#6="root")
    GravityOverrideParams_Tactical_Sprint_57_#23:Gravity Override Setting Struct Structure=(WindFrequency_28_#2=(X=0,Y=0,Z=0),WindAmplitude_31_#3=(X=0,Y=0,Z=0),WindOffset_32_#4=(X=0,Y=0,Z=0),GravityOverride_15_#5=(X=0,Y=-450,Z=-500),JointName_27_#6="root")
    GravityOverrideParams_Rail_Grinding_61_#24:Gravity Override Setting Struct Structure=(WindFrequency_28_#2=(X=0,Y=0,Z=0),WindAmplitude_31_#3=(X=0,Y=0,Z=0),WindOffset_32_#4=(X=0,Y=0,Z=0),GravityOverride_15_#5=(X=0,Y=0,Z=-980),JointName_27_#6="root")
    GravityOverrideParams_DBNO_Tumbling_67_#25:Gravity Override Setting Struct Structure=(WindFrequency_28_#2=(X=0,Y=0,Z=0),WindAmplitude_31_#3=(X=0,Y=0,Z=0),WindOffset_32_#4=(X=0,Y=0,Z=0),GravityOverride_15_#5=(X=-2000,Y=0,Z=0),JointName_27_#6="pelvis")
    GravityOverrideParams_DBNO_RollFwd_68_#26:Gravity Override Setting Struct Structure=(WindFrequency_28_#2=(X=0,Y=0,Z=0),WindAmplitude_31_#3=(X=0,Y=0,Z=0),WindOffset_32_#4=(X=0,Y=0,Z=0),GravityOverride_15_#5=(X=-2000,Y=0,Z=0),JointName_27_#6="pelvis")
    GravityOverrideParams_DBNO_RollBwd_69_#27:Gravity Override Setting Struct Structure=(WindFrequency_28_#2=(X=0,Y=0,Z=0),WindAmplitude_31_#3=(X=0,Y=0,Z=0),WindOffset_32_#4=(X=0,Y=0,Z=0),GravityOverride_15_#5=(X=-4000,Y=0,Z=0),JointName_27_#6="pelvis")
    GravityOverrideParams_DBNO_Impulse_70_#28:Gravity Override Setting Struct Structure=(WindFrequency_28_#2=(X=0,Y=0,Z=0),WindAmplitude_31_#3=(X=0,Y=0,Z=0),WindOffset_32_#4=(X=0,Y=0,Z=0),GravityOverride_15_#5=(X=-2000,Y=0,Z=0),JointName_27_#6="pelvis")
    RBAN Params
    DebugDraw
    DebugJointName
   ) -> (then, GravityOverride, CurrentStateOut, GravityOverrideSettingsOut, RBAN SettingsOut)
   > n2 = Branch(Condition=Get bIsSkyDiving())
   Set Gravity Override(execute=[@n3.else, @n2.then], Gravity Override=@n1)
   > Set External Force(External Force:Vector="0, 0, 0") -> (Output_Get:Vector)
   > Set Scale Clavicle Spr During Skydive(Scale Clavicle Spr During Skydive:double=0.2) -> (Output_Get:double)
   n3 = Branch(execute=@n2.else, Condition=Get bIsFalling())
   > Set Gravity Override(Gravity Override="0,0,-700")
   > Set External Force(External Force="0,0,50")
   > Set Scale Clavicle Spr During Skydive(Scale Clavicle Spr During Skydive=1)
  ]}
]
variables [
 External Force {type=Vector}
 Front Right Pouch Thigh Up {type=Transform, default {location=<0 2 0>, rotation=<rot 0 0 0>, scale=<1 1 1>}}
 Front Right Pouch Thigh Up_0 {type=Transform, default {location=<0 0 0>, rotation=<rot 0 0 0>, scale=<1 1 1>}}
 Gravity Override {type=Vector}
 Gravity Override Params on Ground Falling {
  type=GravityOverrideSettingStruct
  default {jointName=root, gravityOverride=<0 0 600>, windAmplitude=<0 0 400>, windFrequency=<0 0 10>, windOffset=<0 0 400>}
 }
 Gravity Override Params Skydive Floating Down {
  type=GravityOverrideSettingStruct
  default {jointName=root, gravityOverride=<0 -700 600>, windAmplitude=<200 -500 400>, windFrequency=<1.5 7 8>, windOffset=<0 -600 400>}
 }
 Gravity Override Params Skydive Parachuting {
  type=GravityOverrideSettingStruct
  default {jointName=root, gravityOverride=<0 -800 300>, windAmplitude=<200 -500 400>, windFrequency=<1.5 7 8>, windOffset=<0 0 0>}
 }
 Gravity Override Params Ziplining {
  type=GravityOverrideSettingStruct
  default {jointName=root, gravityOverride=<0 0 -980>, windAmplitude=<100 100 100>, windFrequency=<4 8 2>, windOffset=<0 -200 0>}
 }
 Lightsaber Thigh Up {type=Transform, default {location=<-3 0 0>, rotation=<rot -20 -50 0>, scale=<1 1 1>}}
 Sash Thigh Up Front {type=Transform, default {location=<3 0 1>, rotation=<rot 0 0 30>, scale=<1 1 1>}}
 Sash Thigh Up Opposite Front {type=Transform, default {location=<0 2 -2>, rotation=<rot 0 -10 -20>, scale=<1 1 1>}}
 Sash Thigh Up Thigh Side Inner Front Opposite {type=Transform, default {location=<0 0 0>, rotation=<rot 10 0 -10>, scale=<1 1 1>}}
 Scale Clavicle Spr During Skydive {type=number, default=1}
 Thigh Back Rear Sash {type=Transform, default {location=<0 0 0>, rotation=<rot 60 0 0>, scale=<1 1 1>}}
 Thigh Back Underskirt {type=Transform, default {location=<0 0 0>, rotation=<rot 60 0 0>, scale=<1 1 1>}}
 Underskirt Thigh Back Rear {type=Transform, default {location=<0 0 0>, rotation=<rot -15 25 0>, scale=<1 1 1>}}
 Underskirt Thigh Back Rear Center {type=Transform, default {location=<-2 0 2>, rotation=<rot 45 0 0>, scale=<1 1 1>}}
 Underskirt Thigh Up Front {type=Transform, default {location=<6 0 4>, rotation=<rot -20 -5 -10>, scale=<1 1.5 1>}}
 Underskirt Thigh Up Front Center {type=Transform, default {location=<0 0 0>, rotation=<rot 5 0 0>, scale=<1 1 1>}}
 Underskirt Thigh Up Front Scale End {type=Transform, default {location=<0 0 0>, rotation=<rot 0 0 0>, scale=<1 1.75 1>}}
 Underskirt Thigh Up Opposite Front {type=Transform, default {location=<0 0 0>, rotation=<rot 0 0 0>, scale=<1 1 1>}}
 Underskirt Thigh Up Side {type=Transform, default {location=<0 -6 0>, rotation=<rot 20 55 -20>, scale=<1 1 1>}}
 Underskirt Thigh Up Thigh Side Inner Front {type=Transform, default {location=<0 0 0>, rotation=<rot 0 10 0>, scale=<1 1 1>}}
 Upperarm Bk Bk Jnt {type=Transform, default {location=<0 0 0>, rotation=<rot 0 0 0>, scale=<1 1 1>}}
 Upperarm Bk Sash Clavicle {type=Transform, default {location=<7 -5 0>, rotation=<rot -20 0 20>, scale=<1 1 1>}}
 Upperarm Bk Sash Lowerback {type=Transform, default {location=<0 0 4.5>, rotation=<rot 0 20 0>, scale=<1 1 1>}}
 Upperarm Down 45 Sash Lowerback {type=Transform, default {location=<-5 -5 13>, rotation=<rot 0 0 45>, scale=<1 1 1>}}
 Upperarm Down Sash Clavicle {type=Transform, default {location=<-3 -2 0>, rotation=<rot 0 0 20>, scale=<1 1 1>}}
 Upperarm Down Sash Lowerback {type=Transform, default {location=<0 0 0>, rotation=<rot 0 0 40>, scale=<1 1 1>}}
 Upperarm Front Bk Jnt {type=Transform, default {location=<0 0 0>, rotation=<rot 0 0 0>, scale=<1 1 1>}}
 Upperarm Fwd Sash Clavicle {type=Transform, default {location=<0 0 0>, rotation=<rot 0 0 0>, scale=<1 1 1>}}
 Upperarm Raised {type=Transform, default {location=<0 0 8>, rotation=<rot 0 0 0>, scale=<0.7 1 1>}}
 Upperarm Up Sash Clavicle {type=Transform, default {location=<0 -3 6>, rotation=<rot 45 0 0>, scale=<1 1 1>}}
 Upperarm Up Sash Clavicle Skydive {type=Transform, default {location=<0 -3 5>, rotation=<rot 0 0 0>, scale=<1 1 1>}}
]
